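# Physics asset for the DuckCoast pickaxe, defining a jointed chain of dynamic-named bones that hangs off the two-handed melee handle and ends in a larger capsule.
# Small capsules link from the base through the chain, with swing limits tightening toward the tip and little twist, before the end piece gets a wide swing.
# Pelvis, thigh and calf capsules with free angular joints are also included, apparently as body collision for the chain.
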PhysicsAsset /BRCosmetics/Weapons/FORT_Melee/Pickaxe_DuckCoast/Meshes/DuckCoast_Axe_Physics
bodies {
 DuckCoast_01_base [{shapeType=Capsule, length=3, radius=2.5, shapeName=DuckCoast_01_base_capsule, center=<1 0 0>, rotation=<rot 90 0 0>}]
 DuckCoast_01_dyn [{shapeType=Capsule, length=3, radius=2.5, shapeName=DuckCoast_01_dyn_capsule, center=<1 0 0>, rotation=<rot 90 0 0>}]
 DuckCoast_02_dyn [{shapeType=Capsule, length=3, radius=2.5, shapeName=DuckCoast_02_dyn_capsule, center=<1 0 0>, rotation=<rot 90 0 0>}]
 DuckCoast_03_dyn [{shapeType=Capsule, length=3, radius=2.5, shapeName=DuckCoast_03_dyn_capsule, center=<1 0 0>, rotation=<rot 90 0 0>}]
 DuckCoast_04_dyn [{shapeType=Capsule, length=3, radius=2.5, shapeName=DuckCoast_04_dyn_capsule, center=<1 0 0>, rotation=<rot 90 0 0>}]
 DuckCoast_05_dyn [{shapeType=Capsule, length=3, radius=2.5, shapeName=DuckCoast_05_dyn_capsule, center=<1 0 0>, rotation=<rot 90 0 0>}]
 DuckCoast_06_dyn [{shapeType=Capsule, length=3, radius=2.5, shapeName=DuckCoast_06_dyn_capsule, center=<1 0 0>, rotation=<rot 90 0 0>}]
 DuckCoast_07_dyn [{shapeType=Capsule, length=3, radius=2.5, shapeName=DuckCoast_07_dyn_capsule, center=<1 0 0>, rotation=<rot 90 0 0>}]
 DuckCoast_08_dyn [{shapeType=Capsule, length=3, radius=2.5, shapeName=DuckCoast_08_dyn_capsule, center=<1 0 0>, rotation=<rot 90 0 0>}]
 DuckCoast_09_dyn [{shapeType=Capsule, length=18, radius=11, shapeName=DuckCoast_09_dyn_capsule, center=<3.2 -11.5 -10.5>, rotation=<rot 0 90 0>}]
 Melee_TwoHanded_Handle [{shapeType=Capsule, length=4, radius=4}]
 calf_l [{shapeType=Capsule, length=34, radius=9, shapeName=calf_l_capsule, center=<-22 0 0>, rotation=<rot 90 180 0>}]
 calf_r [{shapeType=Capsule, length=34, radius=9, shapeName=calf_r_capsule, center=<22 0 0>, rotation=<rot 90 0 0>}]
 pelvis [{shapeType=Capsule, length=16, radius=15.5, shapeName=pelvis_capsule, center=<0 -2.7681 0>, rotation=<rot 0 -90 0>}]
 thigh_l [{shapeType=Capsule, length=34, radius=9, shapeName=thigh_l_capsule, center=<-22 0 0>, rotation=<rot 90 180 0>}]
 thigh_r [{shapeType=Capsule, length=34, radius=9, shapeName=thigh_r_capsule, center=<22 0 0>, rotation=<rot 90 0 0>}]
}
constraints [
 {
  bone1Name=calf_l
  bone2Name=thigh_l
  swing1LimitDegrees=45
  swing1Motion=Free
  swing2LimitDegrees=45
  swing2Motion=Free
  twistLimitDegrees=45
  twistMotion=Free
 }
 {
  bone1Name=calf_r
  bone2Name=thigh_r
  swing1LimitDegrees=45
  swing1Motion=Free
  swing2LimitDegrees=45
  swing2Motion=Free
  twistLimitDegrees=45
  twistMotion=Free
 }
 {
  bone1Name=thigh_l
  bone2Name=pelvis
  swing1LimitDegrees=45
  swing1Motion=Free
  swing2LimitDegrees=45
  swing2Motion=Free
  twistLimitDegrees=45
  twistMotion=Free
 }
 {
  bone1Name=thigh_r
  bone2Name=pelvis
  swing1LimitDegrees=45
  swing1Motion=Free
  swing2LimitDegrees=45
  swing2Motion=Free
  twistLimitDegrees=45
  twistMotion=Free
 }
 {
  bone1Name=DuckCoast_01_base
  bone2Name=Melee_TwoHanded_Handle
  swing1LimitDegrees=45
  swing1Motion=Free
  swing2LimitDegrees=45
  swing2Motion=Free
  twistLimitDegrees=45
  twistMotion=Free
 }
 {bone1Name=DuckCoast_01_dyn, bone2Name=DuckCoast_01_base, swing1LimitDegrees=45, swing2LimitDegrees=45, twistLimitDegrees=0}
 {bone1Name=DuckCoast_02_dyn, bone2Name=DuckCoast_01_dyn, swing1LimitDegrees=45, swing2LimitDegrees=45, twistLimitDegrees=3}
 {bone1Name=DuckCoast_03_dyn, bone2Name=DuckCoast_02_dyn, swing1LimitDegrees=45, swing2LimitDegrees=45, twistLimitDegrees=3}
 {bone1Name=DuckCoast_04_dyn, bone2Name=DuckCoast_03_dyn, swing1LimitDegrees=40, swing2LimitDegrees=40, twistLimitDegrees=3}
 {bone1Name=DuckCoast_05_dyn, bone2Name=DuckCoast_04_dyn, swing1LimitDegrees=35, swing2LimitDegrees=35, twistLimitDegrees=3}
 {bone1Name=DuckCoast_06_dyn, bone2Name=DuckCoast_05_dyn, swing1LimitDegrees=30, swing2LimitDegrees=30, twistLimitDegrees=3}
 {bone1Name=DuckCoast_07_dyn, bone2Name=DuckCoast_06_dyn, swing1LimitDegrees=25, swing2LimitDegrees=25, twistLimitDegrees=3}
 {bone1Name=DuckCoast_08_dyn, bone2Name=DuckCoast_07_dyn, swing1LimitDegrees=20, swing2LimitDegrees=20, twistLimitDegrees=3}
 {bone1Name=DuckCoast_09_dyn, bone2Name=DuckCoast_08_dyn, swing1LimitDegrees=90, swing2LimitDegrees=90, twistLimitDegrees=0}
]
